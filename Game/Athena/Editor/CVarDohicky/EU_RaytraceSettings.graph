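# EU_RaytraceSettings is an Editor Utility Blueprint that reads and writes the raytraced shadows console variable, r.Raytracing.Shadows.
# The Load event copies the current CVar value into the Shadows boolean.
# The Save event converts Shadows to an integer and passes it to a Set Console Variable macro.
# That macro formats the name and value into a command string and executes it on the editor world.
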EditorUtilityBlueprint /Game/Athena/Editor/CVarDohicky/EU_RaytraceSettings
graphs [
 EventGraph {nodes [
   Load() {is=title K2Node_CustomEvent}
   > n1 = GetConsoleVariableBoolValue!(execute, VariableName:String=r.Raytracing.Shadows) -> (then, ReturnValue:bool)
   > Set Shadows(Shadows:bool=@n1) -> (Output_Get:bool)
   Save() {is=title K2Node_CustomEvent}
   > Sequence()
   > Set Console Variable!(
    execute:Exec
    CVar:String=r.Raytracing.Shadows
    Value:int="To Integer (Boolean)"!(InBool:bool=Get Shadows() -> (Shadows:bool)) -> (ReturnValue:int=0)
   ) -> (then:Exec) {is=title K2Node_MacroInstance}
  ]}
 Set Console Variable {nodes [
   n1 = Inputs!() -> (execute, CVar:String, Value:Wildcard)
   > n2 = GetEditorWorld(self=Get UnrealEditorSubsystem())
   n3 = "To Text (String)"(InString=@n1)
   > ExecuteConsoleCommand(
    WorldContextObject=@n2
    Command="To String (Text)"(InText=Format Text!(Format="LOCTEXT(\"{0} {1}\")", 0:Text=@n3, 1:Wildcard=@n1.Value) -> (Result))
   )
   > Outputs!(then:Exec)
  ]}
]
variables [Shadows {type=boolean}]
