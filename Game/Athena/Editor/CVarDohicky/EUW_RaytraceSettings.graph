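# Editor utility widget that provides a panel for editing ray tracing settings, backed by an EU_RaytraceSettings object.
# Its only widget is a Details View, which generates the editing UI from that object's properties.
# On Construct it creates the settings object, calls Load on it and assigns it to the view.
# Any property change in the view calls Save on the object, so edits are stored immediately.
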EditorUtilityWidgetBlueprint /Game/Athena/Editor/CVarDohicky/EUW_RaytraceSettings
graphs [EventGraph {nodes [
   "On Property Changed (Details)"!() -> (OutputDelegate, then, PropertyName:Name=None) {is=title K2Node_ComponentBoundEvent}
   > Save(self:EU Raytrace Settings Object Reference=Get Settings() -> (Settings:EU Raytrace Settings Object Reference))
   Event Construct()
   > n1 = Construct Object from Class(Class=&/EU_RaytraceSettings.EU_RaytraceSettings_C, self=Self-Reference()) -> (ReturnValue:EU Raytrace Settings Object Reference)
   > n2 = Set Settings(Settings:EU Raytrace Settings Object Reference=@n1) -> (Output_Get:EU Raytrace Settings Object Reference)
   > Load(self:EU Raytrace Settings Object Reference=@n2)
   > SetObject(self=Get Details() -> (Details:Details View Object Reference), NewObject=@n2)
  ]}]
variables [Settings {type=EU_RaytraceSettings_C, default=null}]
widgets {
 info {rootWidgetClass=&DetailsView, widgetCount=1}
 widgets [Details {widgetClassPath=&DetailsView, widget=Details, slot=~}]
}
